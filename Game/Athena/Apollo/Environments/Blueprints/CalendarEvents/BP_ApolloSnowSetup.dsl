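; Apollo map snow controller that drives seasonal ground snow and falling snow from calendar events, with editor override commands.
; On match ready, authority reads WSL1–3 to ramp snow amount up, hold it, or ramp it down.
; Rep-notifies on remotes update material parameters, global snow altitude and post-process volumes.
; Snow amount thresholds switch minimap POI labels to black and a snow font for legibility.
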(event EventBeginPlay
  (Class|FortAsyncActionAuthorityMatchReady|MatchReadyAsync
    (:then
      (CallFunction|RefreshMapLocations))
    (:OnReady
      (CallFunction|RefreshCalender(AuthorityOnly)))))

(event Custom|SetGroundMPCValue
  ;[1 1 1] Set MPC: Holiday_SnowAmount
  (Utilities|FlowControl|Delay 1.0)
  (CallFunction|SetSnowAltitude (Variables|Default|GetSnowAmount))
  (CallFunction|RefreshMapLocations)
  (bind _outactors (Actor|GetAllActorsOfClass "/Game/Athena/Apollo/Environments/Blueprints/CalendarEvents/BP_Snow_PostProcessing_Volume.BP_Snow_PostProcessing_Volume_C"))
  (for _array_element _outactors
    (Class|BPSnowPostProcessingVolume|SetSnowCoveragePP _array_element)))

(event Custom|SetSnowFallMPCValue
  (CallFunction|SetSnowFall (Variables|Default|GetSnowFalling)))

(event Custom|SetSnow
  (params (Intensity :type Float%20%28double-precision%29))
  ;[1 1 1] Editor Commands
  (bind _output_get (Variables|Default|SetSnowAmount Intensity))
  (Variables|Default|SetSnowFalling (Math|Float|Max(Float) _output_get (Variables|Default|GetSnowFalling)))
  (Networking|SwitchHasAuthority
    (:Authority
      (CallFunction|SetSnowAltitude _output_get))) ; Force for Editor
)

(event Custom|SetFullSnowEd
  ;[1 1 1] Editor Commands
  (bind _output_get (Variables|Default|SetSnowAmount 1.0))
  (Variables|Default|SetSnowFalling 1.0)
  (Networking|SwitchHasAuthority
    (:Authority
      (CallFunction|SetSnowAltitude _output_get))) ; Force for Editor
)

(event Custom|SetFallingSnow
  (params (Intensity :type Float%20%28double-precision%29))
  ;[1 1 1] Editor Commands
  (Variables|Default|SetSnowFalling Intensity))

(event Custom|SetStartSnowEd
  ;[1 1 1] Editor Commands
  (bind _output_get (Variables|Default|SetSnowAmount 0.0))
  (Networking|SwitchHasAuthority
    (:Authority
      (CallFunction|SetSnowAltitude _output_get))) ; Force for Editor
)

(event Custom|RefreshCalender%20%28Authority%20Only%29
  (Networking|SwitchHasAuthority
    (:Authority
      (Utilities|FlowControl|Sequence
        (:then_0
          (bind (_biseventactive _timesincebegin _timeuntilend _timespanratio) (CalendarEvents|GetCalendarEventInformation "WSL1"))
          (if _biseventactive
            (Variables|Default|SetSnowAmount _timespanratio)
            (else
              (bind _biseventactive_2 (CalendarEvents|GetCalendarEventInformation "WSL2"))
              (if _biseventactive_2
                (Variables|Default|SetSnowAmount 1.0)
                (else
                  (bind (_biseventactive_3 _timesincebegin_2 _timeuntilend_2 _timespanratio_2) (CalendarEvents|GetCalendarEventInformation "WSL3"))
                  (if _biseventactive_3
                    (Variables|Default|SetSnowAmount (- 1.0 _timespanratio_2))
                    (else
                      (Variables|Default|SetSnowAmount 0.0))))))))))))

(event Custom|RefreshMapLocations
  (if (not (Variables|Default|GetPendingMinimapRefresh))
    (Variables|Default|SetPendingMinimapRefresh true)
    (bind _gamestate (Class|FortAsyncActionGameStateReady|GameStateReadyAsync)
      (:OnReady
        (bind _asfort_game_state_athena (Utilities|Casting|CastToFortGameStateAthena _gamestate))
        (Variables|Default|SetBRGameState _asfort_game_state_athena)
        (Variables|Default|SetPendingMinimapRefresh false)
        ;[1 1 1] Update Minimap Text
        (Utilities|FlowControl|Sequence
          (:then_0
            (if (>= (Variables|Default|GetSnowAmount) 0.15)
              (CallFunction|MakeLocationVisibleinSnow (GameplayTags|MakeLiteralGameplayTag "(TagName=\"Athena.Location.POI.RetailRow\")"))))
          (:then_1
            (if (>= (Variables|Default|GetSnowAmount) 0.35)
              (CallFunction|MakeLocationVisibleinSnow (GameplayTags|MakeLiteralGameplayTag "(TagName=\"Athena.Location.POI.LazyLake\")"))))
          (:then_2
            (if (>= (Variables|Default|GetSnowAmount) 0.4)
              (CallFunction|MakeLocationVisibleinSnow (GameplayTags|MakeLiteralGameplayTag "(TagName=\"Athena.Location.POI.MountainMeadow\")"))))
          (:then_3
            (if (>= (Variables|Default|GetSnowAmount) 0.625)
              (CallFunction|MakeLocationVisibleinSnow (GameplayTags|MakeLiteralGameplayTag "(TagName=\"Athena.Location.POI.HollyHedges\")"))))
          (:then_4
            (if (>= (Variables|Default|GetSnowAmount) 0.675)
              (CallFunction|MakeLocationVisibleinSnow (GameplayTags|MakeLiteralGameplayTag "(TagName=\"Athena.Location.POI.SaltySprings\")"))
              (CallFunction|MakeLocationVisibleinSnow (GameplayTags|MakeLiteralGameplayTag "(TagName=\"Athena.Location.POI.FrenzyFarm\")"))
              (CallFunction|MakeLocationVisibleinSnow (GameplayTags|MakeLiteralGameplayTag "(TagName=\"Athena.Location.POI.PowerPlant\")"))))
          (:then_5
            (if (>= (Variables|Default|GetSnowAmount) 0.75)
              (CallFunction|MakeLocationVisibleinSnow (GameplayTags|MakeLiteralGameplayTag "(TagName=\"Athena.Location.POI.PleasantPark\")"))))
          (:then_6
            (if (>= (Variables|Default|GetSnowAmount) 0.8)
              (CallFunction|MakeLocationVisibleinSnow (GameplayTags|MakeLiteralGameplayTag "(TagName=\"Athena.Location.POI.WeepingWoods\")"))
              (CallFunction|MakeLocationVisibleinSnow (GameplayTags|MakeLiteralGameplayTag "(TagName=\"Athena.Location.POI.BeachyBluffs\")"))))
          (:then_7
            (if (>= (Variables|Default|GetSnowAmount) 0.85)
              (CallFunction|MakeLocationVisibleinSnow (GameplayTags|MakeLiteralGameplayTag "(TagName=\"Athena.Location.POI.SlurpySwamp\")"))
              (CallFunction|MakeLocationVisibleinSnow (GameplayTags|MakeLiteralGameplayTag "(TagName=\"Athena.Location.POI.SunnyShores\")"))
              (CallFunction|MakeLocationVisibleinSnow (GameplayTags|MakeLiteralGameplayTag "(TagName=\"Athena.Location.POI.DirtyDocks\")")))))))))

(fn MakeLocationVisibleInSnow
  (params (LocationTag :type Gameplay%20Tag%20Structure%20%28by%20ref%29))
  (UI|Minimap|SetLocationColorinState (Variables|Default|GetBRGameState) (Variables|GetLocationTag) (Utilities|Struct|MakeLinearColor 0.0 0.0 0.0 1.0) "Normal")
  (UI|Minimap|SetLocationFontinState (Variables|Default|GetBRGameState) (Variables|GetLocationTag) (Variables|Default|GetSnowFont) "Normal"))

(fn OnRep_Snow_Amount
  (Networking|SwitchHasAuthority
    (:Remote
      (Variables|Default|SetDidEverRepNotify true)
      (CallFunction|SetGroundMPCValue))))

(fn OnRep_SnowFalling
  (Networking|SwitchHasAuthority
    (:Remote
      (CallFunction|SetSnowFallMPCValue))))

(fn ProxyCalendarEventRatio
  (params (Start%20Time :type Date%20Time%20Structure) (End%20Time :type Date%20Time%20Structure) (EventTime :type Date%20Time%20Structure))
  (return (Math|Float|Clamp(Float) (Math|Timespan|TimespanRatio (Math|DateTime|DateTime-DateTime EventTime Start%20Time) (Math|DateTime|DateTime-DateTime End%20Time Start%20Time))) (Math|Float|Truncate (Math|Timespan|GetTotalHours (Math|DateTime|DateTime-DateTime End%20Time Start%20Time))) (Utilities|String|Append "Ratio is: " (Utilities|String|ToString(Float) (Math|Float|Clamp(Float) (Math|Timespan|TimespanRatio (Math|DateTime|DateTime-DateTime EventTime Start%20Time) (Math|DateTime|DateTime-DateTime End%20Time Start%20Time)))) ", Event hour length is: " (Utilities|String|ToString(Integer) (Math|Float|Truncate (Math|Timespan|GetTotalHours (Math|DateTime|DateTime-DateTime End%20Time Start%20Time)))) ", Event is" (select (and (Math|DateTime|DateTime>=DateTime EventTime Start%20Time) (Math|DateTime|DateTime<DateTime EventTime End%20Time)) "Active" "Not_Active")) (and (Math|DateTime|DateTime>=DateTime EventTime Start%20Time) (Math|DateTime|DateTime<DateTime EventTime End%20Time))))

(fn SetSnowAltitude
  (params (SnowAmount :type Float%20%28double-precision%29))
  (bind _snowaltitudeproxy (Variables|Default|GetSnowAltitudeProxy))
  (Rendering|Material|SetScalarParameterValue "/Game/Packages/Fortress_SharedMaterials/GlobalMaterialParameters/FortniteMaterialParameters.FortniteMaterialParameters" "SnowCoverage" SnowAmount :!class /Script/Engine.KismetMaterialLibrary)
  (bind _output_get (Variables|Default|SetCurrentSnowAltitude (+ (Math|Float|MapRangeClamped SnowAmount 0.0 1.0 (Variables|Default|GetSnow_StartHeight) (Variables|Default|GetSnow_EndHeight)) (Variables|Default|GetSnowAltitudeOffset))))
  (FortGlobals|SetGlobalSnowAltitude _output_get)
  (FortGlobals|SetGlobalSnowOrigin (Variables|Default|GetSnowOrigin))
  (FortGlobals|SetGlobalSnowRadius (Variables|Default|GetSnowRadius))
  ;[1 1 1] Debug
  (if (Variables|Default|GetShowDebug)
    (Rendering|SetVisibility _snowaltitudeproxy true false)
    (Transformation|SetWorldLocation (Variables|Default|GetSnowAltitudeProxy) (Math|Vector|MakeVector (.x (Transformation|GetWorldLocation (Variables|Default|GetSnowAltitudeProxy))) (.y (Transformation|GetWorldLocation (Variables|Default|GetSnowAltitudeProxy))) (Variables|Default|GetCurrentSnowAltitude)) false false)
    (else
      (Rendering|SetVisibility _snowaltitudeproxy false false))))

(fn SetSnowFall
  (params (FallingSnow :type Float%20%28double-precision%29))
  (Rendering|Material|SetScalarParameterValue "/Game/Packages/Fortress_SharedMaterials/GlobalMaterialParameters/FortniteMaterialParameters.FortniteMaterialParameters" "FallingSnowCoverage" (Math|Curves|GetFloatValue (Variables|Default|GetCurve) FallingSnow) :!class /Script/Engine.KismetMaterialLibrary))

(fn ConstructionScript
  (CallFunction|SetSnowAltitude (Variables|Default|GetSnowAmount))
  (CallFunction|SetSnowFall (Variables|Default|GetSnowFalling)))
