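; Calendar-event Blueprint volume that sizes a snow post-process area to the global SnowCoverage material parameter.
; Construction Script and SetSnowCoveragePP event run identical logic, setting PostProcess blend radius and Sphere radius.
; Coverage is remapped from a start offset, then lerps between minimum and maximum radii, minus half the blend radius.
; A Ceil multiplier acts as a gate on that radius; BeginPlay, Tick, overlap and OnRep_Update are empty.
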(event Custom|SetSnowCoveragePP
  (bind _blend_radius (Variables|Default|GetBlend%20Radius))
  (bind _returnvalue (Rendering|Material|GetScalarParameterValue "/Game/Packages/Fortress_SharedMaterials/GlobalMaterialParameters/FortniteMaterialParameters.FortniteMaterialParameters" "SnowCoverage" :!class /Script/Engine.KismetMaterialLibrary))
  (Variables|/Script/Engine.PostProcessComponent|SetBlendRadius _blend_radius (Variables|Default|GetPostProcess))
  (Components|Sphere|SetSphereRadius (Variables|Default|GetSphere) (- (* (Math|Float|Ceil (+ (* (/ (- (Math|Float|Clamp(Float) _returnvalue) (Variables|Default|GetOffset%20Start%20Point)) (- 1.0 (Variables|Default|GetOffset%20Start%20Point))) (- 1.0 (Variables|Default|Getfloat%20value))) (Variables|Default|Getfloat%20value))) (Math|Float|Lerp (Variables|Default|GetSnow%20PostProccesing%20Min%20Raduis) (Variables|Default|GetSnow%20PostProccesing%20Max%20Raduis) (Math|Float|Clamp(Float) (+ (* (/ (- (Math|Float|Clamp(Float) _returnvalue) (Variables|Default|GetOffset%20Start%20Point)) (- 1.0 (Variables|Default|GetOffset%20Start%20Point))) (- 1.0 (Variables|Default|Getfloat%20value))) (Variables|Default|Getfloat%20value))))) (* _blend_radius 0.5)) true))

(event EventBeginPlay)

(event Collision|EventActorBeginOverlap
  (params (OtherActor :type Actor%20Object%20Reference)))

(event EventTick
  (params (DeltaSeconds :type Float%20%28single-precision%29)))

(fn OnRep_Update)

(fn ConstructionScript
  (bind _blend_radius (Variables|Default|GetBlend%20Radius))
  (bind _returnvalue (Rendering|Material|GetScalarParameterValue "/Game/Packages/Fortress_SharedMaterials/GlobalMaterialParameters/FortniteMaterialParameters.FortniteMaterialParameters" "SnowCoverage" :!class /Script/Engine.KismetMaterialLibrary))
  (Variables|/Script/Engine.PostProcessComponent|SetBlendRadius _blend_radius (Variables|Default|GetPostProcess))
  (Components|Sphere|SetSphereRadius (Variables|Default|GetSphere) (- (* (Math|Float|Ceil (+ (* (/ (- (Math|Float|Clamp(Float) _returnvalue) (Variables|Default|GetOffset%20Start%20Point)) (- 1.0 (Variables|Default|GetOffset%20Start%20Point))) (- 1.0 (Variables|Default|Getfloat%20value))) (Variables|Default|Getfloat%20value))) (Math|Float|Lerp (Variables|Default|GetSnow%20PostProccesing%20Min%20Raduis) (Variables|Default|GetSnow%20PostProccesing%20Max%20Raduis) (Math|Float|Clamp(Float) (+ (* (/ (- (Math|Float|Clamp(Float) _returnvalue) (Variables|Default|GetOffset%20Start%20Point)) (- 1.0 (Variables|Default|GetOffset%20Start%20Point))) (- 1.0 (Variables|Default|Getfloat%20value))) (Variables|Default|Getfloat%20value))))) (* _blend_radius 0.5)) true))
