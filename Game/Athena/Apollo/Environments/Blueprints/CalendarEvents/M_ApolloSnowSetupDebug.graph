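# Translucent, two-sided debug material for the Apollo calendar event snow setup, likely used to visualize where its surface meets scene geometry.
# An inverted depth fade, with parameterized opacity and distance, gives values that rise near intersecting geometry.
# That value is clamped between OpacityMin and OpacityMax, and separately blends ColorA to ColorB, both yellow by default.
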Material /Game/Athena/Apollo/Environments/Blueprints/CalendarEvents/M_ApolloSnowSetupDebug
graph {
 properties {blendMode=BLEND_Translucent, pixelDepthOffsetMode=PDOM_Legacy, twoSided=true, usedShadingModels=Default Lit}
 nodes [
  n1 = OneMinus(None=DepthFade(
    Opacity=ScalarParameter() {defaultValue=0.5, parameterName=Opacity}
    FadeDistance=ScalarParameter() {defaultValue=464.6113, parameterName=FadeDistance}
   ))
  Clamp(
   None=@n1
   Min=ScalarParameter() {defaultValue=0.4, parameterName=OpacityMin}
   Max=ScalarParameter() {defaultValue=1, parameterName=OpacityMax}
  )
  LinearInterpolate(
   A=VectorParameter() {parameterName=ColorA, defaultValue=<rgba 0.948 1 0 1>}.RGB
   B=VectorParameter() {parameterName=ColorB, defaultValue=<rgba 0.948 1 0 1>}.RGB
   Alpha=Clamp(None=@n1)
  )
 ]
}
